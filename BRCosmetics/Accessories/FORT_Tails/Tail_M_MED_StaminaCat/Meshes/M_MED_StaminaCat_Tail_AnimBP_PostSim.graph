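# Post-sim animation blueprint for the StaminaCat cosmetic tail, adjusting its pose for long back bling and sliding.
# Each update sets isWearingCape from the long-backbling check, switching between the input pose and a tucked-tail animation.
# Additive bone rotations shape the tail before a trail controller adds lagging motion along the chain, both weighted by TrailAlpha.
# Further additive rotations after the trail re-pose the tail while bIsSliding is true.
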AnimBlueprint /BRCosmetics/Accessories/FORT_Tails/Tail_M_MED_StaminaCat/Meshes/M_MED_StaminaCat_Tail_AnimBP_PostSim
target_skeleton=/Game/Accessories/FORT_Tails/Common/Fortnite_Base_Tail_Skeleton
graphs [
 AnimGraph {nodes [
   Copy Pose From Mesh()
   n1 = Input Pose()
   n2 = Random Sequence Player!() -> (Pose:Pose Link Structure) {
    is=title AnimGraphNode_RandomPlayer
    entries [{
      chanceToPlay=1
      maxLoopCount=0
      maxPlayRate=1
      minLoopCount=0
      minPlayRate=1
      sequence=/Game/Accessories/FORT_Tails/Tail_M_MED_StaminaCat/Animations/StaminaCat_Tucked_Tail
      blendIn {blendOption=Linear, blendTime=0.2, customCurve=null}
     }]
   }
   n3 = Get isWearingCape() -> (isWearingCape:bool)
   n4 = Get TrailAlpha() -> (TrailAlpha:double)
   n5 = "Transform (Modify) Bone - Bone: C_Tail_A_Base_Jnt"(
    ComponentPose=Local To Component(LocalPose=Blend Poses by bool(BlendPose_0=@n2, BlendPose_1=@n1, bActiveValue=@n3) {blendType=Linear})
    Alpha=@n4
    Translation="0,0,0"
    Rotation="-8,-72,23"
    Scale="1,1,1"
   ) {
    rotationMode=BMM_Additive
    rotationSpace=BCS_BoneSpace
    alphaBoolBlend {blendOption=Linear}
    boneToModify {boneName=C_Tail_A_Base_Jnt}
    rotation=<rot -8 -72 23>
   }
   n6 = "Transform (Modify) Bone - Bone: C_Tail_A_1_Jnt"(ComponentPose=@n5, Alpha=@n4, Translation="0,0,0", Rotation="-31,0,-0.000", Scale="1,1,1") {
    rotationMode=BMM_Additive
    rotationSpace=BCS_BoneSpace
    alphaBoolBlend {blendOption=Linear}
    boneToModify {boneName=C_Tail_A_1_Jnt}
    rotation {pitch=-31}
   }
   n7 = "Transform (Modify) Bone - Bone: C_Tail_A_2_Jnt"(ComponentPose=@n6, Alpha=@n4, Translation="0,0,0", Rotation="-4,0.000,0.000", Scale="1,1,1") {
    rotationMode=BMM_Additive
    rotationSpace=BCS_BoneSpace
    alphaBoolBlend {blendOption=Linear}
    boneToModify {boneName=C_Tail_A_2_Jnt}
    rotation {pitch=-4}
   }
   n8 = "Transform (Modify) Bone - Bone: C_Tail_A_3_Jnt"(ComponentPose=@n7, Alpha=@n4, Translation="0,0,0", Rotation="0,0,0", Scale="1,1,1") {rotationMode=BMM_Additive, rotationSpace=BCS_BoneSpace, alphaBoolBlend {blendOption=Linear}, boneToModify {boneName=C_Tail_A_3_Jnt}}
   n9 = "Transform (Modify) Bone - Bone: C_Tail_A_4_Jnt"(ComponentPose=@n8, Alpha=@n4, Translation="0,0,0", Rotation="0,0,0", Scale="1,1,1") {rotationMode=BMM_Additive, rotationSpace=BCS_BoneSpace, alphaBoolBlend {blendOption=Linear}, boneToModify {boneName=C_Tail_A_4_Jnt}}
   n10 = "Transform (Modify) Bone - Bone: C_Tail_A_5_Jnt"(ComponentPose=@n9, Alpha=@n4, Translation="0,0,0", Rotation="0,0,0", Scale="1,1,1") {rotationMode=BMM_Additive, rotationSpace=BCS_BoneSpace, alphaBoolBlend {blendOption=Linear}, boneToModify {boneName=C_Tail_A_5_Jnt}}
   n11 = "Transform (Modify) Bone - Bone: C_Tail_A_6_Jnt"(ComponentPose=@n10, Alpha=@n4, Translation="0,0,0", Rotation="15,0.000,0.000", Scale="1,1,1") {
    rotationMode=BMM_Additive
    rotationSpace=BCS_BoneSpace
    alphaBoolBlend {blendOption=Linear}
    boneToModify {boneName=C_Tail_A_6_Jnt}
    rotation {pitch=15}
   }
   n12 = "Transform (Modify) Bone - Bone: C_Tail_A_7_Jnt"(ComponentPose=@n11, Alpha=@n4, Translation="0,0,0", Rotation="7,0.000,0.000", Scale="1,1,1") {
    rotationMode=BMM_Additive
    rotationSpace=BCS_BoneSpace
    alphaBoolBlend {blendOption=Linear}
    boneToModify {boneName=C_Tail_A_7_Jnt}
    rotation {pitch=7}
   }
   n13 = "Transform (Modify) Bone - Bone: C_Tail_A_8_Jnt"(ComponentPose=@n12, Alpha=@n4, Translation="0,0,0", Rotation="50,0.000,-0.000", Scale="1,1,1") {
    rotationMode=BMM_Additive
    rotationSpace=BCS_BoneSpace
    alphaBoolBlend {blendOption=Linear}
    boneToModify {boneName=C_Tail_A_8_Jnt}
    rotation {pitch=50}
   }
   n14 = "Transform (Modify) Bone - Bone: C_Tail_A_9_Jnt"(ComponentPose=@n13, Alpha=@n4, Translation="0,0,0", Rotation="37,0.000,-0.000", Scale="1,1,1") {
    rotationMode=BMM_Additive
    rotationSpace=BCS_BoneSpace
    alphaBoolBlend {blendOption=Linear}
    boneToModify {boneName=C_Tail_A_9_Jnt}
    rotation {pitch=37}
   }
   n15 = Trail controller - Bone: C_Tail_A_9_Jnt!(ComponentPose=@n14, Alpha=Get TrailAlpha(), RelaxationSpeedScale:float=Get RelaxSpeed() -> (RelaxSpeed:double)) -> (Pose) {
    bLimitRotation=true
    chainLength=10
    maxDeltaTime=0.0333
    alphaScaleBiasClamp {bInterpResult=true}
    baseJoint {boneName=C_Tail_A_Base_Jnt}
    rotationLimits [
     {limitMax=<180 180 180>}
     {limitMax=<180 180 180>}
     {limitMax=<180 180 180>}
     {limitMax=<180 180 180>}
     {limitMax=<180 180 180>}
     {limitMax=<180 180 180>}
     {limitMax=<180 180 180>}
     {limitMax=<180 180 180>}
     {limitMax=<180 180 180>}
     {limitMax=<180 180 180>}
    ]
    rotationOffsets [<0 0 0>, <0 0 0>, <0 0 0>, <0 0 0>, <0 0 0>, <0 0 0>, <0 0 0>, <0 0 0>, <0 0 0>, <0 0 0>]
    trailBone {boneName=C_Tail_A_9_Jnt}
    trailRelaxationSpeed {editorCurveData {keys [<at 0 10 RCIM_Linear>,<at 0.5107 8.1304 RCIM_Linear>,<at 1 5 RCIM_Linear>]}}
   }
   n16 = Get bIsSliding()
   n17 = "Transform (Modify) Bone - Bone: C_Tail_A_Base_Jnt"!(ComponentPose=@n15, bAlphaBoolEnabled:bool=@n16, Translation="0,0,0", Rotation="0,80,0", Scale="1,1,1") -> (Pose) {
    alphaInputType=Bool
    rotationMode=BMM_Additive
    rotationSpace=BCS_BoneSpace
    alphaBoolBlend {blendOption=Linear}
    boneToModify {boneName=C_Tail_A_Base_Jnt}
    rotation {yaw=80}
   }
   n18 = "Transform (Modify) Bone - Bone: C_Tail_A_1_Jnt"!(ComponentPose=@n17, bAlphaBoolEnabled:bool=@n16, Translation="0,0,0", Rotation="30,0,0", Scale="1,1,1") -> (Pose) {
    alphaInputType=Bool
    rotationMode=BMM_Additive
    rotationSpace=BCS_BoneSpace
    alphaBoolBlend {blendOption=Linear}
    boneToModify {boneName=C_Tail_A_1_Jnt}
    rotation {pitch=30}
   }
   n19 = "Transform (Modify) Bone - Bone: C_Tail_A_2_Jnt"!(ComponentPose=@n18, bAlphaBoolEnabled:bool=@n16, Translation="0,0,0", Rotation="40,0,0", Scale="1,1,1") -> (Pose) {
    alphaInputType=Bool
    rotationMode=BMM_Additive
    rotationSpace=BCS_BoneSpace
    alphaBoolBlend {blendOption=Linear}
    boneToModify {boneName=C_Tail_A_2_Jnt}
    rotation {pitch=40}
   }
   n20 = "Transform (Modify) Bone - Bone: C_Tail_A_6_Jnt"!(ComponentPose=@n19, bAlphaBoolEnabled:bool=@n16, Translation="0,0,0", Rotation="-30,0,0", Scale="1,1,1") -> (Pose) {
    alphaInputType=Bool
    rotationMode=BMM_Additive
    rotationSpace=BCS_BoneSpace
    alphaBoolBlend {blendOption=Linear}
    boneToModify {boneName=C_Tail_A_6_Jnt}
    rotation {pitch=-30}
   }
   Output Pose(Result=Component To Local(ComponentPose=@n20)) {layerGroup=None}
  ]}
 EventGraph {nodes [
   Event BlueprintUpdateAnimation()
   > Set isWearingCape(isWearingCape=CheckForBackbling().bIsBackblingLong)
   Event OnCharacterCustomizationLoaded()
  ]}
]
variables [
 isWearingCape {type=boolean}
 RelaxSpeed {type=number}
 TrailAlpha {type=number}
]
